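# Post-process AnimBP for the male medium MuffinLadle GasChin cosmetic head, layering its head Control Rig over the incoming pose.
# The rig's alpha defaults to full, and in the front end follows the absolute difference between the FrontEnd_Idle_Bypass curve and one.
# Each update also stores a scaled, negated owner head-socket rotation for an additive head-bone node.
# That bone node and a Copy Pose From Mesh node are disconnected from the output.
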AnimBlueprint /BRCosmetics/Characters/Player/Male/Medium/Heads/M_MED_MuffinLadle_GasChin_Head/Meshes/M_MED_MuffinLadle_GasChin_Post_Process_AnimBP
target_skeleton=/Game/Characters/Player/Common/Fortnite_Base_Head/Fortnite_Base_Head_Export_Skeleton
graphs [
 AnimGraph {nodes [
   n1 = "Transform (Modify) Bone - Bone: head"(ComponentPose=Local To Component(), Alpha=1, Translation="0,0,0", Rotation=Get Head_Rot() -> (Head_Rot:Rotator), Scale="1,1,1") {rotationMode=BMM_Additive, rotationSpace=BCS_WorldSpace, alphaBoolBlend {blendOption=Linear}, boneToModify {boneName=head}}
   Component To Local(ComponentPose=@n1)
   Copy Pose From Mesh() {bUseAttachedParent=true}
   n2 = Control Rig(Source=Input Pose(), Alpha=Get Head Rot Alpha() -> (Head Rot Alpha:double)) {
    alphaBoolBlend {blendOption=Linear}
    controlRigAssetReference {blueprintRigClass=&/M_MED_MuffinLadle_GasChin_Head_CtrlRig.M_MED_MuffinLadle_GasChin_Head_CtrlRig_C}
   }
   Output Pose(Result=@n2) {layerGroup=None}
  ]}
 EventGraph {nodes [
   Event BlueprintUpdateAnimation()
   n1 = BreakRotator(InRot=GetSocketRotation!(self:Scene Component Object Reference=GetOwnerSkeletalMeshComponent(), InSocketName:Name=head) -> (ReturnValue:Rotator))
   n2 = Get Rotate_Percentage() -> (Rotate_Percentage:double)
   > Set Head_Rot(Head_Rot:Rotator=MakeRotator(
     Roll=float * float!(A=@n1, B=-1, C:double=@n2) -> (ReturnValue) {is=title K2Node_PromotableOperator}
     Pitch=float * float(A=@n1.Pitch, B=-1, C=@n2)
     Yaw=float * float(A=@n1.Yaw, B=-1, C=@n2)
    )) -> (Output_Get:Rotator)
   > Branch(Condition=Get bIsInFrontEnd())
   n3 = GetCurveValue(self=GetOwnerFrontEndAnimInstance(), CurveName=FrontEnd_Idle_Bypass)
   > Set Head Rot Alpha(Head Rot Alpha:double="Absolute (Float)"(A=float - float(A=@n3, B=1))) -> (Output_Get:double)
  ]}
]
variables [
 Head Rot Alpha {type=number, default=1}
 Head_Rot {type=Rotator}
 Rotate_Percentage {type=number, default=0.5}
]
